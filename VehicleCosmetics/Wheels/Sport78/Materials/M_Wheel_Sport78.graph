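# Masked wheel material for the Sport78 vehicle cosmetic, adding animated gradient color and glow effects to tire and rim.
# MF_WheelParent_Simple supplies base attributes, tire and rim channel masks, and a vertex-color tire mask.
# Panning stripe textures, one shaped by Fresnel, build gradients from custom and accent colors, with named reroutes feeding base color and emissive blends.
# The result passes through MF_BaseVehicle_Damage, then MF_VehicleDeRez_Transition.
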Material /VehicleCosmetics/Wheels/Sport78/Materials/M_Wheel_Sport78
graph {
 properties {
  bAutomaticallySetUsageInEditor=false
  bUseMaterialAttributes=true
  bUsedWithMeshDeformer=true
  bUsedWithSkeletalMesh=true
  blendMode=BLEND_Masked
  pixelDepthOffsetMode=PDOM_Legacy
  usedShadingModels=Default Lit
 }
 nodes [
  n1 = MaterialFunctionCall() {materialFunction=&MF_WheelParent_Simple}
  n2 = GetMaterialAttributes(None=@n1.MaterialAttributes) {attributeGetTypes [
    #1
    #2
    #3
    #4
   ]}
  n3 = LinearInterpolate(
   A=LinearInterpolate(A=@n2.Base Color, B=NamedRerouteUsage(), Alpha=NamedRerouteUsage())
   B=LinearInterpolate(A=@n2.Base Color, B=NamedRerouteUsage(), Alpha=NamedRerouteUsage())
   Alpha=NamedRerouteUsage()
  )
  n4 = Multiply(A=NamedRerouteUsage(), B=ScalarParameter() {defaultValue=5, group=5_Special Effects, parameterName=GradientEmissiveValue, sortPriority=37})
  n5 = LinearInterpolate(A=@n2.Emissive Color, B=@n4, Alpha=NamedRerouteUsage())
  n6 = Multiply(A=NamedRerouteUsage(), B=ScalarParameter() {defaultValue=5, group=5_Special Effects, parameterName=CircleEmissiveValue, sortPriority=36})
  n7 = NamedRerouteUsage()
  n8 = VectorParameter() {group=5_Special Effects, parameterName=EmissiveColor, sortPriority=35, defaultValue=<rgba 0.0962 0.9043 0.7955 1>}
  n9 = LinearInterpolate(
   A=LinearInterpolate(A=LinearInterpolate(A=@n5, B=@n6, Alpha=@n7), B=@n8.RGB, Alpha=NamedRerouteUsage())
   B=LinearInterpolate(A=@n2.Emissive Color, B=@n4, Alpha=NamedRerouteUsage())
   Alpha=NamedRerouteUsage()
  )
  MaterialFunctionCall(In=MaterialFunctionCall(Material=SetMaterialAttributes!(MaterialAttributes=@n2.MaterialAttributes, Base Color=@n3, Emissive Color=@n9, Normal) {attributeSetTypes [#1, #4, #3]}) {materialFunction=&MF_BaseVehicle_Damage}.Result) {materialFunction=&MF_VehicleDeRez_Transition}
  n10 = NamedRerouteUsage()
  n11 = Panner(Time=Multiply(A=Time(), B=ScalarParameter() {defaultValue=0.75, group=5_Special Effects, parameterName=Speed, sortPriority=38})) {speedX=-0.5}
  n12 = Power(Base=Multiply(A=Fresnel(Normal=@n10), B=TextureSample(UVs=@n11) {texture=/VehicleCosmetics/Wheels/_SharedTextures/FX/T_GradientStripeVertical_M}.RGB))
  NamedRerouteDeclaration(None=LinearInterpolate(A=NamedRerouteUsage(), B=NamedRerouteUsage(), Alpha=SmoothStep(Min=@n12) {constValue=0.7})) {name=GradientColor, nodeColor=<rgba 0 0.9062 1 1>}
  NamedRerouteDeclaration(None=ComponentMask(None=@n1.TireRGBAMask) {r=true}) {name=TireRMask, nodeColor=<rgba 1 0.7734 0 1>}
  n13 = NamedRerouteUsage()
  NamedRerouteDeclaration(None=Multiply(A=NamedRerouteDeclaration(None=@n2.Normal) {name=Normal, nodeColor=<rgba 0 0.7187 1 1>}, B=OneMinus(None=@n13))) {name=RimNormal, nodeColor=<rgba 0.8516 0 1 1>}
  NamedRerouteDeclaration(None=ComponentMask(None=@n1.TireRGBAMask) {g=true}) {name=TireGMask, nodeColor=<rgba 0.4531 0 1 1>}
  NamedRerouteDeclaration(None=ComponentMask(None=@n1.RimRGBAMask) {a=true}) {name=RimAMask, nodeColor=<rgba 0.1484 0 1 1>}
  n14 = VectorParameter() {group=5_Special Effects, parameterName=CustomColor, sortPriority=34, defaultValue=<rgba 0.0442 0.0807 0.6348 1>}
  NamedRerouteDeclaration(None=@n14.RGB) {name=GradientColor1, nodeColor=<rgba 1 0.3047 0 1>}
  n15 = VectorParameter() {group=5_Special Effects, parameterName=AccentColor, sortPriority=33, defaultValue=<rgba 0.0597 0.0051 0.1478 1>}
  NamedRerouteDeclaration(None=@n15.RGB) {name=GradientColor2, nodeColor=<rgba 0 0.0156 1 1>}
  n16 = NamedRerouteUsage()
  n17 = Constant3Vector()
  n18 = Constant() {r=0.5}
  n19 = Panner(Coordinate=TextureCoordinate() {uTiling=1.5}, Time=Time()) {speedX=-0.5}
  NamedRerouteDeclaration(None=LinearInterpolate(
    A=NamedRerouteUsage()
    B=LinearInterpolate(A=@n16, B=@n17, Alpha=@n18)
    Alpha=TextureSample(UVs=@n19) {texture=&T_GradientStripeVertical_M}.RGB
   )) {name=CircleGradientBase, nodeColor=<rgba 1 0.9141 0 1>}
  NamedRerouteDeclaration(None=ComponentMask(None=@n1.TireRGBAMask) {b=true}) {name=TireBMask, nodeColor=<rgba 0 1 0.1328 1>}
  NamedRerouteDeclaration(None=@n1.VertexColorR) {name=VertexTireMask, nodeColor=<rgba 1 0 0.1875 1>}
  NamedRerouteDeclaration(None=ComponentMask(None=@n1.RimRGBAMask) {r=true}) {name=RimRMask, nodeColor=<rgba 1 0 0.1172 1>}
  NamedRerouteDeclaration(None=ComponentMask(None=@n1.RimRGBAMask) {g=true}) {name=RimGMask, nodeColor=<rgba 0 0.4375 1 1>}
  NamedRerouteDeclaration(None=ComponentMask(None=@n1.RimRGBAMask) {b=true}) {name=RimBMask, nodeColor=<rgba 0.7578 1 0 1>}
 ]
}
