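# Special-effects material for the Sport78 vehicle wheel cosmetic, a Default Lit surface for skeletal meshes that drives only emissive.
# One rim mask texture is sampled unshifted and at three UV offsets from the local-space camera vector, faking parallax depth layers.
# Those samples blend and add parameterized colors, highlight and boosted glow.
# The "Fresnel" glow comes from mask values, including a rescaled, view-shifted sample, rather than a Fresnel node.
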Material /VehicleCosmetics/Wheels/Sport78/Materials/M_Wheel_Sport78_FX
graph {
 properties {
  bUseMaterialAttributes=true
  bUsedWithMeshDeformer=true
  bUsedWithSkeletalMesh=true
  pixelDepthOffsetMode=PDOM_Legacy
  usedShadingModels=Default Lit
 }
 nodes [
  n1 = TextureCoordinate() {uTiling=0.7, vTiling=0.7}
  n2 = Constant2Vector() {g=0.15, r=0.15}
  n3 = Transform(None=CameraVectorWS()) {transformSourceType=TRANSFORMSOURCE_World, transformType=TRANSFORM_Local}
  n4 = DotProduct(A=@n3, B=Constant3Vector() {constant=<rgba 1 0 0 0>})
  n5 = AppendVector(A=Multiply(A=ComponentMask(None=@n4) {r=true}) {constB=0.4}, B=Multiply(A=ComponentMask(None=@n3) {b=true}) {constB=-0.4})
  n6 = Constant() {r=-0.2}
  n7 = Constant() {r=0.2}
  n8 = TextureSampleParameter2D(UVs=Add(A=Add(A=@n1, B=@n2), B=Clamp(None=@n5, Min=@n6, Max=@n7) {maxDefault=0.2, minDefault=-0.2})) {
   group=0_SpecialEffects
   parameterName=Mask
   samplerType=SAMPLERTYPE_LinearColor
   sortPriority=0
   texture=/VehicleCosmetics/Wheels/Sport78/Textures/T_Rim_Sport78_Mask
  }
  n9 = Multiply(A=@n8.G, B=ScalarParameter() {defaultValue=0.35, group=0_SpecialEffects, parameterName=Fresnel_Intensity, sortPriority=6})
  n10 = SmoothStep(Min=Constant() {r=0.15}, Max=Constant() {r=2.2}, Value=NamedRerouteUsage()) {constMin=0.25}
  n11 = VectorParameter() {group=0_SpecialEffects, parameterName=FresnelColor, sortPriority=5, defaultValue=<rgba 0.0194 0.2195 0.2747 1>}
  n12 = Multiply(A=@n11.RGB, B=Add(A=Add(A=@n9, B=@n10), B=Multiply(A=@n11.RGB) {constB=0.01}))
  n13 = VectorParameter() {group=0_SpecialEffects, parameterName=Color1, sortPriority=1, defaultValue=<rgba 0.5391 0.5391 0.5391 1>}
  n14 = VectorParameter() {group=0_SpecialEffects, parameterName=HighlightColor, sortPriority=4, defaultValue=<rgba 0.2462 0.7682 0.7084 1>}
  n15 = Transform(None=CameraVectorWS()) {transformSourceType=TRANSFORMSOURCE_World, transformType=TRANSFORM_Local}
  n16 = Constant3Vector() {constant=<rgba 1 0 0 0>}
  n17 = Multiply(A=ComponentMask(None=DotProduct(A=@n15, B=@n16)) {r=true}) {constB=-0.1}
  n18 = Multiply(A=ComponentMask(None=@n15) {b=true}) {constB=0.1}
  n19 = TextureSampleParameter2D(UVs=Add(A=TextureCoordinate(), B=AppendVector(A=@n17, B=@n18))) {
   group=0_SpecialEffects
   parameterName=Mask
   samplerType=SAMPLERTYPE_LinearColor
   sortPriority=0
   texture=&T_Rim_Sport78_Mask
  }
  n20 = VectorParameter() {group=0_SpecialEffects, parameterName=Color2, sortPriority=2, defaultValue=<rgba 0.521 0.0742 1 1>}
  n21 = LinearInterpolate(A=Multiply(A=@n19.R, B=@n20.RGB), B=LinearInterpolate(A=@n13.RGB, B=@n14.RGB, Alpha=NamedRerouteUsage()), Alpha=NamedRerouteUsage())
  n22 = Add(A=NamedRerouteUsage(), B=@n19.R)
  n23 = VectorParameter() {group=0_SpecialEffects, parameterName=Color3, sortPriority=3, defaultValue=<rgba 0.0097 0.0369 0.956 1>}
  n24 = Transform(None=CameraVectorWS()) {transformSourceType=TRANSFORMSOURCE_World, transformType=TRANSFORM_Local}
  n25 = Constant3Vector() {constant=<rgba 1 0 0 0>}
  n26 = Multiply(A=ComponentMask(None=DotProduct(A=@n24, B=@n25)) {r=true}) {constB=-0.2}
  n27 = Multiply(A=ComponentMask(None=@n24) {b=true}) {constB=0.2}
  n28 = TextureSampleParameter2D(UVs=Add(A=TextureCoordinate(), B=AppendVector(A=@n26, B=@n27))) {
   group=0_SpecialEffects
   parameterName=Mask
   samplerType=SAMPLERTYPE_LinearColor
   sortPriority=0
   texture=&T_Rim_Sport78_Mask
  }
  n29 = Multiply(A=OneMinus(None=Clamp(None=@n22)), B=Multiply(A=@n23.RGB, B=@n28.R) {constB=10})
  MakeMaterialAttributes!(
   BaseColor
   Metallic
   Specular
   Roughness
   Anisotropy
   EmissiveColor=Add(A=Multiply(A=@n12) {constB=5}, B=Add(A=@n21, B=@n29))
   Opacity
   OpacityMask
   Normal
   Tangent
   WorldPositionOffset
   SubsurfaceColor
   ClearCoat
   ClearCoatRoughness
   AmbientOcclusion
   Refraction
   CustomizedUVs_0
   CustomizedUVs_1
   CustomizedUVs_2
   CustomizedUVs_3
   CustomizedUVs_4
   CustomizedUVs_5
   CustomizedUVs_6
   CustomizedUVs_7
   PixelDepthOffset
   ShadingModel
   Displacement
  )
  n30 = TextureSampleParameter2D() {
   group=0_SpecialEffects
   parameterName=Mask
   samplerType=SAMPLERTYPE_LinearColor
   sortPriority=0
   texture=&T_Rim_Sport78_Mask
  }
  NamedRerouteDeclaration(None=@n30.R) {name=MaskR, nodeColor=<rgba 0 1 0.3672 1>}
  NamedRerouteDeclaration(None=@n30.G) {name=MaskG, nodeColor=<rgba 1 0.0469 0 1>}
  NamedRerouteDeclaration(None=@n30.B) {name=maskB, nodeColor=<rgba 0 0.2734 1 1>}
  NamedRerouteDeclaration(None=@n30.A) {name=maskA, nodeColor=<rgba 0.1719 0 1 1>}
 ]
}
